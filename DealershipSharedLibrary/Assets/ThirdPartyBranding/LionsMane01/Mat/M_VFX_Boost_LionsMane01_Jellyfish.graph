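# Unlit translucent material for the jellyfish in the LionsMane01 branded boost VFX, usable on Niagara sprites and mesh particles.
# Fresnel-weighted texture alpha plus brightness-scaled texture color feeds a shared emissive adjustments function.
# Time-scrolled sine and cosine waves along UV red build an offset, transformed between local and world space.
# Vertex color RGB weights that offset with differing signs and strengths into a "WPO" named reroute.
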Material /DealershipSharedLibrary/Assets/ThirdPartyBranding/LionsMane01/Mat/M_VFX_Boost_LionsMane01_Jellyfish
graph {
 properties {
  bAutomaticallySetUsageInEditor=false
  bUsedWithNiagaraMeshParticles=true
  bUsedWithNiagaraSprites=true
  blendMode=BLEND_Translucent
  blendableLocation=BL_ReplacingTonemapper
  shadingModel=MSM_Unlit
  usedShadingModels=Unlit
 }
 nodes [
  n1 = TextureSampleParameter2D(UVs=TextureCoordinate()) {parameterName=Param, texture=/DealershipSharedLibrary/Assets/ThirdPartyBranding/LionsMane01/Tex/T_VCVFX_LionsMane01_Jellyfish_diffuse}
  n2 = Multiply(A=@n1.A, B=Constant() {r=1.2})
  n3 = Power(Base=Fresnel(ExponentIn=Constant() {r=2}))
  n4 = ScalarParameter() {defaultValue=0.5, parameterName=Brightness}
  MaterialFunctionCall(Emissive=Add(A=Multiply(A=@n2, B=@n3), B=Multiply(A=@n4, B=@n1.RGB)) {constB=0}) {materialFunction=&MF_FX_EmissiveAdjustments}
  n5 = TextureCoordinate()
  n6 = VectorParameter() {group=WPO, parameterName=WPO Parameters, channelNames {a=WPO Strength 2, b=WPO Strength, g=Speed, r=Tile}, defaultValue=<rgba 1.5 1 8 -5>}
  n7 = Add(
   A=Divide(A=ComponentMask(None=@n5) {r=true}, B=@n6.Tile)
   B=MaterialFunctionCall(Speed=@n6.Speed) {materialFunction=&TimeWithSpeedVariable}.Time
  )
  n8 = Cosine(None=@n7)
  n9 = Sine(None=@n7)
  n10 = Multiply(
   A=Transform(None=MaterialFunctionCall(X=@n8, Y=@n9, Z=@n9) {materialFunction=&MakeFloat3}.Result) {transformSourceType=TRANSFORMSOURCE_Local}
   B=@n6.WPO Strength
  )
  n11 = ConstantBiasScale(None=Transform(None=Multiply(A=@n10, B=@n6.WPO Strength 2)) {transformSourceType=TRANSFORMSOURCE_World, transformType=TRANSFORM_Local})
  n12 = Constant3Vector() {constant=<rgba 1 -0.5 -1 0>}
  n13 = Multiply(A=@n11, B=@n12.R)
  n14 = VertexColor()
  n15 = Multiply(A=Multiply(A=@n11, B=@n12.G), B=@n14.G)
  n16 = Multiply(A=Multiply(A=@n11, B=@n12.B), B=@n14.B)
  NamedRerouteDeclaration(None=Add(A=Multiply(A=@n13, B=@n14.R), B=Add(A=@n15, B=@n16))) {name=WPO, nodeColor=<rgba 0.6875 1 0 1>}
 ]
}
